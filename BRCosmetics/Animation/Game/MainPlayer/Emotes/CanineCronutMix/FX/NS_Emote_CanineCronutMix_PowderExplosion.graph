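# Niagara system for the green powder explosion in the CanineCronutMix emote.
# Burst emitters layer outward and upward smoke poofs and streaks, alongside a smoke ribbon trail.
# A single flare/dust/decal emitter partitions particles by visibility tag across sprite and decal renderers, with separate per-layer age, size and color attributes.
# Colors adapt to time of day or front-end preview, and BlendSpace switches world/local positioning by owner speed.
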NiagaraSystem /BRCosmetics/Animation/Game/MainPlayer/Emotes/CanineCronutMix/FX/NS_Emote_CanineCronutMix_PowderExplosion
dependencies {
 usedDataInterfaces [
  &NiagaraDataInterfaceArrayColor
  &NiagaraDataInterfaceMeshRendererInfo
  &NiagaraDataInterfaceCurve
  &NiagaraDataInterfaceVectorCurve
  &NiagaraDataInterfaceVector2DCurve
  &NiagaraDataInterfaceColorCurve
  &NiagaraDataInterfacePlatformSet
 ]
 usedDynamicInputs []
 usedModules [&SystemState, &BlendSpaceHelper]
}
emitters {
 FallingSmokeDust {rendererClasses [], renderers []}
 Flare_Dusty_Decal_ReVamped {
  emitterUpdateScript [
   &EmitterState()
   &SpawnBurst_Instantaneous(Spawn Count=8, Use Spawn Probability=false, UseLoopCountLimit=false)
  ]
  particleSpawnScript [
   &PartitionParticles(
    Particles Per Partition=1
    Partition Count Mode:ENiagara_InfiniteLoopDuration=::NewEnumerator1
    Partition Mode:ENiagara_PartitionMode_v2=::NewEnumerator1
   )
   /SetVariables(Particles.VisibilityTag=$Particles.PartitionParticles.Partition) {isSetParameters=true}
   &InitializeParticle(
    Color=&FNTimeOfDayColor(
     Night Color=<rgba 0.0454 0.1615 0.0933 1>
     Morning Color=<rgba 0.077 0.219 0.098 1>
     Day Color=<rgba 0.0775 0.2188 0.0979 1>
     Evening Color=<rgba 0.0775 0.2188 0.0979 1>
    )
    Lifetime=1.35
    Position Offset=&MergeHelper_VectorSwitch(0=<0 0 30>, 1=<0 0 100>, 2=<0 0 20>, 3=<0 0 120>, Index=$Particles.PartitionParticles.Partition, Num=::NewEnumerator2)
    Sprite Rotation Angle=180
    Sprite Rotation Mode=::NewEnumerator0
    Sprite Size=<180 200>
    Sprite Size Mode=::NewEnumerator4
    UsePositionOffset=true
    Write Material Random=true
   )
   &CameraOffset()
   &AddVelocity(Module.Use Falloff Distance=false, Origin Offset=<0 0 -17>, Override Seed=false, Velocity Mode=::NewEnumerator1, Velocity Speed=50) {enabled=false}
   /SetVariables(Particles.Flare Color:LinearColor=<rgba 0.1678 1 0.546 1>, Particles.Flare Lifetime=0.16, Particles.Flare Sprite Size:Vector2f=<350 350>) {isSetParameters=true}
   /SetVariables(
    Particles.Dust Color:LinearColor=&FNTimeOfDayColor(
     Night Color=<rgba 0.0808 0.2188 0.185 1>
     Morning Color=<rgba 0.0775 0.2188 0.0979 1>
     Day Color=<rgba 0.0775 0.2188 0.0979 1>
     Evening Color=<rgba 0.0775 0.2188 0.0979 1>
    )
    Particles.Dust Lifetime=0.5
    Particles.Dust Sprite Size:Vector2f=<350 300>
   ) {isSetParameters=true}
  ]
  particleUpdateScript [
   &ParticleState()
   /SetVariables(
    Particles.Update Dust Age:NiagaraFloat=&Divide_Float(A=$Particles.Age, B=$Particles.Initial.Dust Lifetime)
    Particles.Update Flare Age:NiagaraFloat=&Divide_Float(A=$Particles.Age, B=$Particles.Initial.Flare Lifetime)
   ) {isSetParameters=true}
   /SetVariables(
    Particles.Update Dust Sprite Color:LinearColor=&MakeLinearColorFromVectorAndFloat(
     "Vector (RGB)"=<0.077 0.219 0.098>
     "Float (Alpha)"=&FloatFromCurve(FloatCurve=DataInterface() {propertyValues {Curve {keys [
          <at 0.2338 0 RCIM_Cubic>
          {arriveTangentWeight=0.1559, interpMode=RCIM_Cubic, leaveTangentWeight=0.1827, tangentMode=RCTM_User, time=0.3156, value=0.08}
          {arriveTangentWeight=0.1824, interpMode=RCIM_Cubic, leaveTangentWeight=0.1903, tangentMode=RCTM_User, time=0.4565, value=0.08}
          {interpMode=RCIM_Cubic, leaveTangent=0.9867, time=0.71, value=0}
         ]}}})
    )
    Particles.Update Dust Sprite Size:Vector2f=&Multiply_Vector2D_ByFloat(
     Float=&FloatFromCurve(
      FloatCurve=DataInterface() {propertyValues {Curve {
         defaultValue=0
         keys [
          {arriveTangent=4.1785, interpMode=RCIM_Cubic, leaveTangent=4.1785, tangentMode=RCTM_User, time=0, value=0}
          {arriveTangent=1.1179, interpMode=RCIM_Cubic, leaveTangent=1.1179, tangentMode=RCTM_User, time=0.2042, value=0.6228}
          {arriveTangent=0.0224, interpMode=RCIM_Cubic, leaveTangent=0.0224, tangentMode=RCTM_User, time=1, value=1}
         ]
        }}}
      CurveIndex=$Particles.Update Dust Age
     )
     Vector2D=$Particles.Initial.Dust Sprite Size
    )
    Particles.Update Flare Sprite Color:LinearColor=&MakeLinearColorFromVectorAndFloat(
     "Vector (RGB)"=<0.1678 1 0.546>
     "Float (Alpha)"=&MakeFloatFromLinearColor(
      LinearColor=/Niagara/DynamicInputs/TypeConversions/MakeLinearColorFromFloat(Value:NiagaraFloat=&FloatFromCurve(
        FloatCurve=DataInterface() {propertyValues {Curve {keys [{interpMode=RCIM_Cubic, tangentMode=RCTM_User, time=0, value=1}, {interpMode=RCIM_Cubic, tangentMode=RCTM_User, time=1, value=0}]}}}
        CurveIndex=$Particles.Update Flare Age
        Scale Curve=&FNTimeOfDayScalar(Night Scale=0.5, Morning Scale=0.5, Day Scale=1.5, Evening Scale=0.5)
       ))
      Channel:ENiagaraFloatFromLinearColorOptions=::NewEnumerator0
     )
    )
    Particles.Update Flare Sprite Size:Vector2f=&Multiply_Vector2D_ByFloat(
     Float=&FloatFromCurve(
      FloatCurve=DataInterface() {propertyValues {Curve {keys [{interpMode=RCIM_Cubic, tangentMode=RCTM_User, time=0, value=1}, {interpMode=RCIM_Cubic, tangentMode=RCTM_User, time=1, value=0}]}}}
      CurveIndex=$Particles.Update Flare Age
     )
     Vector2D=$Particles.Initial.Flare Sprite Size
    )
   ) {isSetParameters=true}
   /SetVariables(
    Particles.DecalFade:NiagaraFloat=&FloatFromCurve(
     FloatCurve=DataInterface() {propertyValues {Curve {keys [
         <at 0.2338 0 RCIM_Cubic>
         {arriveTangentWeight=0.1824, interpMode=RCIM_Cubic, leaveTangentWeight=0.1903, tangentMode=RCTM_User, time=0.4114, value=1}
         {arriveTangentWeight=0.1413, interpMode=RCIM_Cubic, leaveTangentWeight=0.2764, tangentMode=RCTM_User, time=0.75, value=1}
         {interpMode=RCIM_Cubic, leaveTangent=0.9867, time=1, value=0}
        ]}}}
     Scale Curve=2
    )
    Particles.DecalSize:Vector3f=&Multiply_Vector(
     A=<20 460 460>
     B=&VectorFromFloat(Value=&FloatFromCurve(FloatCurve=DataInterface() {propertyValues {Curve {keys [
           {arriveTangent=4.1785, interpMode=RCIM_Cubic, leaveTangent=4.1785, tangentMode=RCTM_User, time=0, value=0.5}
           {arriveTangent=4.0435, interpMode=RCIM_Cubic, leaveTangent=4.0435, tangentMode=RCTM_User, time=0.0096, value=0.95}
           {arriveTangent=0.0224, interpMode=RCIM_Cubic, leaveTangent=0.0224, tangentMode=RCTM_User, time=0.05, value=1}
          ]}}}))
    )
   ) {isSetParameters=true}
   /SetVariables(Particles.Dust SubImageIndex:NiagaraFloat=&FloatFromCurve(FloatCurve=DataInterface() {propertyValues {Curve {keys [<at 0 30>,<at 1 64>]}}})) {isSetParameters=true}
   &ScaleSpriteSize(Scale Factor=&Vector2DFromFloat(Value=&FloatFromCurve(
      FloatCurve=DataInterface() {propertyValues {Curve {keys [
          {arriveTangent=0.6836, interpMode=RCIM_Cubic, leaveTangent=0.6836, tangentMode=RCTM_User, time=-0.0093, value=0.5}
          {arriveTangent=0.4954, interpMode=RCIM_Cubic, leaveTangent=0.4954, time=0.2196, value=0.85}
          {arriveTangent=0.1863, arriveTangentWeight=0.3375, interpMode=RCIM_Cubic, leaveTangent=0.1863, tangentMode=RCTM_User, time=1, value=1}
         ]}}}
      Scale Curve=0.9
     )))
   &DynamicMaterialParameters(
    Index 0 Param 1=3
    Index 0 Param 2=10
    Module.Index0Float0Write=true
    Module.Index0Float1Write=true
    Module.Index0Float2Write=true
    Module.Index0Float3Write=true
   )
   &ScaleColor(
    Scale Alpha=&FloatFromCurve(FloatCurve=DataInterface() {propertyValues {Curve {keys [
         {arriveTangent=-1.1408, interpMode=RCIM_Cubic, leaveTangent=-1.1408, tangentMode=RCTM_User, time=0, value=0}
         {
          arriveTangent=-0.4785
          arriveTangentWeight=0.1274
          interpMode=RCIM_Cubic
          leaveTangent=-0.4785
          leaveTangentWeight=0.2616
          tangentMode=RCTM_User
          time=0.0922
          value=1
         }
         {
          arriveTangent=-0.4785
          arriveTangentWeight=0.2192
          interpMode=RCIM_Cubic
          leaveTangent=-0.4785
          leaveTangentWeight=0.2077
          tangentMode=RCTM_User
          time=0.2195
          value=1
         }
         {arriveTangent=-0.2671, interpMode=RCIM_Cubic, leaveTangent=-0.2671, tangentMode=RCTM_User, time=1.012, value=0}
        ]}}})
    ScaleA=true
    ScaleRGB=true
   )
   /SetVariables(Particles.DynamicMaterialParameterDustySmoke:Vector4f=&MakeVector4(X=3, Y=1, Z=1, W=0)) {isSetParameters=true}
   &BlendSpace("Position Blend (World <-> Local)":NiagaraFloat=&MakeCustomFloatFromBool(
     Bool=&SetBoolByFloatComparison(A=&VectorLength(Vector=$Engine.Owner.Velocity), B=800)
     True Float=1
     False Float=0
     Translation Contribution:NiagaraFloat=1
     Rotation Contribution:NiagaraFloat=1
    ))
   &SolveForcesAndVelocity(Clamp Velocity=false, Limit Acceleration=false)
  ]
  rendererClasses [&NiagaraSpriteRendererProperties, &NiagaraDecalRendererProperties]
  renderers [
   {rendererClass=&NiagaraSpriteRendererProperties}
   {rendererClass=&NiagaraSpriteRendererProperties, rendererIndex=1}
   {rendererClass=&NiagaraDecalRendererProperties, rendererIndex=2}
   {rendererClass=&NiagaraSpriteRendererProperties, rendererIndex=3}
  ]
 }
 Smoke_Poof_Out {
  emitterUpdateScript [
   &EmitterState()
   &SpawnBurst_Instantaneous(Spawn Count=12, Use Spawn Probability=false)
  ]
  particleSpawnScript [
   /SetVariables(Particles.VelCurveRand:NiagaraFloat=&FloatFromCurve(
     FloatCurve=DataInterface() {propertyValues {Curve {keys [
         {interpMode=RCIM_Cubic, leaveTangent=0.1132, tangentMode=RCTM_Break, time=0, value=0}
         {arriveTangent=1.9858, interpMode=RCIM_Cubic, leaveTangent=1.7203, tangentMode=RCTM_Break, time=0.3, value=0.4}
         {interpMode=RCIM_Cubic, tangentMode=RCTM_Break, time=1, value=1}
        ]}}}
     CurveIndex=&RandomRangeFloat(Minimum=0.9, Maximum=1)
     Scale Curve=10
    )) {isSetParameters=true, module=SetVariables_#1}
   /SetVariables(Particles.InitSpriteRotRate=&RandomRangeFloat(Minimum=-1, Maximum=1)) {isSetParameters=true, module=SetVariables_#2}
   &InitializeParticle(
    Color=&SelectLinearColorFromArray(
     Array Sampling Mode=::NewEnumerator0
     Direct Array Index=?
     Color Selection Array=DataInterface() {propertyValues {ColorData [
        <rgba 0.0802 0.3385 0.2024 1>
        <rgba 0.0775 0.2188 0.0979 1>
        <rgba 0.0959 0.3542 0.1332 1>
        <rgba 0.0775 0.2188 0.0979 1>
        <rgba 0.1154 0.3021 0.0822 1>
        <rgba 0.0685 0.2865 0.0683 1>
       ]}}
     Random Seed=0
    )
    Lifetime Max=0.7
    Lifetime Min=0.6
    Lifetime Mode=::NewEnumerator1
    Position Offset=<0 6.0689 0>
    Position Offset Coordinate Space=::Local
    Sprite Rotation Angle=&RandomRangeFloat(Minimum=0, Maximum=360)
    Sprite Rotation Mode=::NewEnumerator0
    Sprite Size Mode=::NewEnumerator3
    Uniform Sprite Size=110
    UsePositionOffset=true
    Write Material Random=true
   )
   /SetVariables(Particles.SpriteSize=&Multiply_Vector2D_ByFloat(Float=1.7, Vector2D=$Particles.SpriteSize)) {isSetParameters=true, module=SetVariables_#3}
   &CameraOffset(Camera Offset Amount=&UniformRangedFloat(Minimum=15, Maximum=20))
   &ShapeLocation(
    Handle Radius=0.1
    Large Radius=1
    MaskSpawnGroup=false
    Offset=<0 0 10>
    Offset Mode=::NewEnumerator0
    Override Seed=false
    Rotation Coordinate Space=::World
    Shape Primitive=::NewEnumerator3
    UseOwnerScale=true
   )
   &AddVelocityFromPoint(
    Module.Use Falloff Distance=false
    Origin Offset=<0 5.3831 10>
    Velocity Strength=&Lerp_Float(A=101.0153, B=132.9134, Alpha=$Particles.VelCurveRand)
   )
  ]
  particleUpdateScript [
   &ParticleState()
   &Drag(Drag=8, Use Linear Drag=true, Use Rotational Drag=false)
   &SpriteRotationRate(Rotation Rate=&Multiply_Float(
     A=$Particles.InitSpriteRotRate
     B=&FloatFromCurve(
      FloatCurve=DataInterface() {propertyValues {Curve {keys [
          {arriveTangent=-4.4624, interpMode=RCIM_Cubic, leaveTangent=-4.4624, tangentMode=RCTM_User, time=0, value=1}
          {arriveTangentWeight=0.1718, interpMode=RCIM_Cubic, tangentMode=RCTM_User, time=0.5, value=0.15}
         ]}}}
      Scale Curve=1000
     )
    ))
   &ScaleColor(
    Scale Alpha=&FloatFromCurve(FloatCurve=DataInterface() {propertyValues {Curve {keys [
         {arriveTangent=-1.3772, interpMode=RCIM_Cubic, leaveTangent=-1.3772, tangentMode=RCTM_User, time=0, value=0}
         {arriveTangentWeight=0.1274, interpMode=RCIM_Cubic, leaveTangentWeight=0.2616, tangentMode=RCTM_User, time=0.0922, value=1}
         {arriveTangentWeight=0.2192, interpMode=RCIM_Cubic, leaveTangentWeight=0.2077, tangentMode=RCTM_User, time=0.3704, value=1}
         {arriveTangent=-2.5885, interpMode=RCIM_Cubic, leaveTangent=-2.5885, tangentMode=RCTM_User, time=1.012, value=0}
        ]}}})
    Scale RGB=&MakeVectorFromLinearColorRGB(Color=&RandomRangeLinearColor(
      Minimum=&ColorFromCurve(ColorCurve=DataInterface() {propertyValues {
         AlphaCurve {keys []}
         BlueCurve {keys [<at 0.0654 5>,<at 0.1422 1 RCIM_Cubic>]}
         GreenCurve {keys [<at 0.0654 5>,<at 0.1422 1 RCIM_Cubic>]}
         RedCurve {keys [<at 0.0654 5>,<at 0.1422 1 RCIM_Cubic>]}
        }})
      Maximum=<rgba 1 1 1 1>
     ))
    ScaleA=true
    ScaleRGB=true
   )
   &ScaleSpriteSize(Scale Factor=&Vector2DFromFloat(Value=&FloatFromCurve(
      FloatCurve=DataInterface() {propertyValues {Curve {keys [
          {arriveTangent=5.0612, interpMode=RCIM_Cubic, leaveTangent=5.0612, tangentMode=RCTM_User, time=0, value=0.6}
          {arriveTangent=0.2572, interpMode=RCIM_Cubic, leaveTangent=0.2572, tangentMode=RCTM_User, time=0.1894, value=0.95}
          {arriveTangent=0.016, arriveTangentWeight=0.3375, interpMode=RCIM_Cubic, leaveTangent=0.016, tangentMode=RCTM_User, time=1, value=1}
         ]}}}
      Scale Curve=0.9
     )))
   &GravityForce(Gravity=<0 0 100>)
   &SolveForcesAndVelocity(Clamp Velocity=false, Limit Acceleration=false)
   /SetVariables(Particles.Position=$Particles.Position) {enabled=false, isSetParameters=true}
   &DynamicMaterialParameters(
    Index 0 Param 1=&Add_Float(
     A=&RandomRangeFloat(Minimum=0, Maximum=1)
     B=&FloatFromCurve(
      FloatCurve=DataInterface() {propertyValues {Curve {keys [
          {arriveTangent=2.7533, interpMode=RCIM_Cubic, leaveTangent=1.8958, tangentMode=RCTM_User, time=0, value=1}
          {arriveTangent=0.8292, interpMode=RCIM_Cubic, leaveTangent=0.8292, tangentMode=RCTM_User, time=0.321, value=1.3453}
          {arriveTangent=0.0529, interpMode=RCIM_Cubic, leaveTangent=0.2021, tangentMode=RCTM_User, time=1, value=1.5}
         ]}}}
      CurveIndex=&Add_Float(A=&Multiply_Float(A=$Particles.Initial.MaterialRandom, B=0.5), B=$Particles.NormalizedAge)
     )
    )
    Index 0 Param 2=&FloatFromCurve(FloatCurve=DataInterface() {propertyValues {Curve {keys [
         {arriveTangent=0.0453, interpMode=RCIM_Cubic, tangentMode=RCTM_Break, time=0, value=-1}
         {arriveTangent=0.0604, interpMode=RCIM_Cubic, leaveTangent=0.0604, tangentMode=RCTM_User, time=1, value=-1}
        ]}}})
    Index 0 Param 3=&FloatFromCurve(
     FloatCurve=DataInterface() {propertyValues {Curve {keys [
         {arriveTangent=0.0632, interpMode=RCIM_Cubic, leaveTangent=0.0632, tangentMode=RCTM_User, time=0, value=0.5}
         {
          arriveTangent=-1.6858
          arriveTangentWeight=0.0333
          interpMode=RCIM_Cubic
          leaveTangent=-1.6858
          leaveTangentWeight=0.3137
          tangentMode=RCTM_User
          time=0.02
          value=0.1
         }
         {arriveTangentWeight=0.1027, interpMode=RCIM_Cubic, tangentMode=RCTM_Break, time=0.2, value=0}
        ]}}}
     Scale Curve=&RandomRangeFloat(Minimum=0.25, Maximum=0.5)
    )
    Index 0 Param 4=&FloatFromCurve(FloatCurve=DataInterface() {propertyValues {Curve {keys [<at 0.5 1>,<at 1 0>]}}})
    Index 1 Param 1=&FloatFromCurve(
     FloatCurve=DataInterface() {propertyValues {Curve {keys [
         {interpMode=RCIM_Cubic, tangentMode=RCTM_Break, time=0, value=1}
         {arriveTangentWeight=0.0833, interpMode=RCIM_Cubic, tangentMode=RCTM_Break, time=0.25, value=0.3}
        ]}}}
     Scale Curve=&RandomRangeFloat(Minimum=1.5, Maximum=3)
    )
    Index 1 Param 2=&FloatFromCurve(
     FloatCurve=DataInterface() {propertyValues {Curve {keys [
         {interpMode=RCIM_Cubic, leaveTangent=-2.5489, tangentMode=RCTM_Break, time=0, value=1}
         {interpMode=RCIM_Cubic, leaveTangent=0.9066, tangentMode=RCTM_Break, time=1, value=0}
        ]}}}
     Scale Curve=&RandomRangeFloat(Minimum=0.01, Maximum=0.1)
    )
    Index 1 Param 3=&FloatFromCurve(FloatCurve=DataInterface() {propertyValues {Curve {keys [
         {interpMode=RCIM_Cubic, leaveTangent=6.3917, tangentMode=RCTM_Break, time=0, value=0.2}
         {arriveTangent=1.7549, interpMode=RCIM_Cubic, leaveTangent=10.6767, tangentMode=RCTM_User, time=0.25, value=1}
        ]}}})
    Index 1 Param 4=&FloatFromCurve(
     FloatCurve=DataInterface() {propertyValues {Curve {keys [{arriveTangent=-0.136, interpMode=RCIM_Cubic, leaveTangent=-0.136, tangentMode=RCTM_User, time=0, value=1}, <at 1 0.95>]}}}
     Scale Curve=&RandomRangeFloat(Minimum=1.3, Maximum=2)
    )
    Index 2 Param 1=&FloatFromCurve(FloatCurve=DataInterface() {propertyValues {Curve {keys [
         {interpMode=RCIM_Cubic, leaveTangent=4.2485, tangentMode=RCTM_Break, time=0, value=0}
         {arriveTangent=1.0548, interpMode=RCIM_Cubic, leaveTangent=1.0548, tangentMode=RCTM_User, time=0.17, value=0.5715}
         <at 1 1>
        ]}}})
    Index 2 Param 2=&FloatFromCurve(
     FloatCurve=DataInterface() {propertyValues {Curve {keys [
         {arriveTangent=1.9056, interpMode=RCIM_Cubic, leaveTangent=1.9056, tangentMode=RCTM_User, time=0, value=0.2}
         {arriveTangent=0.0259, interpMode=RCIM_Cubic, leaveTangent=0.0259, tangentMode=RCTM_User, time=1, value=1.4}
        ]}}}
     Scale Curve=&RandomRangeFloat(Minimum=0.1, Maximum=0.2)
    )
    Index 2 Param 3=&FloatFromCurve(
     FloatCurve=DataInterface() {propertyValues {Curve {keys [
         {interpMode=RCIM_Cubic, leaveTangent=0.0393, tangentMode=RCTM_Break, time=0, value=1}
         {interpMode=RCIM_Cubic, leaveTangent=11.1233, tangentMode=RCTM_Break, time=1, value=1.5}
        ]}}}
     Scale Curve=0.7
    )
    Index 2 Param 4=&FloatFromCurve(
     FloatCurve=DataInterface() {propertyValues {Curve {keys [
         {arriveTangent=-0.6867, interpMode=RCIM_Cubic, leaveTangent=-0.6867, tangentMode=RCTM_User, time=0, value=1}
         {interpMode=RCIM_Cubic, leaveTangent=-1.2956, tangentMode=RCTM_Break, time=1, value=0.75}
        ]}}}
     Scale Curve=1.5
    )
    Module.Index0Float0Write=true
    Module.Index0Float1Write=true
    Module.Index0Float2Write=true
    Module.Index0Float3Write=true
    Module.Index1Float0Write=true
    Module.Index1Float1Write=true
    Module.Index1Float2Write=true
    Module.Index1Float3Write=true
    Module.Index2Float0Write=true
    Module.Index2Float1Write=true
    Module.Index2Float2Write=true
    Module.Index2Float3Write=true
    Write Parameter Index 1=::NewEnumerator1
    Write Parameter Index 2=::NewEnumerator1
   )
   &ScaleColor(
    Scale Alpha=&FN_FrontEndFloatModifier(Gameplay=1, Front End=0.9)
    Scale RGB=&MakeVectorFromLinearColorRGB(Color=&MakeCustomLinearColorFromBool(
      Bool=$User.bIsFrontEndPreview
      True Color=&RandomRangeLinearColor(Minimum=<rgba 0.4 0.55 1 1>, Maximum=<rgba 0.08 0.11 0.2 1>)
      False Color=<rgba 1 1 1 1>
     ))
    ScaleA=true
    ScaleRGB=true
   ) {module=ScaleColor001}
   &BlendSpace("Position Blend (World <-> Local)":NiagaraFloat=&MakeCustomFloatFromBool(
     Bool=&SetBoolByFloatComparison(A=&VectorLength(Vector=$Engine.Owner.Velocity), B=800)
     True Float=1
     False Float=0
     Translation Contribution:NiagaraFloat=1
     Rotation Contribution:NiagaraFloat=&MakeCustomFloatFromBool(
      Bool=&SetBoolByFloatComparison(A=&VectorLength(Vector=$Engine.Owner.Velocity), B=800)
      True Float=1
      False Float=0
     )
    ))
   /SetVariables() {isSetParameters=true}
  ]
  rendererClasses [&NiagaraSpriteRendererProperties]
  renderers [{rendererClass=&NiagaraSpriteRendererProperties}]
 }
 Smoke_Poof_Up {
  emitterSpawnScript [/SetVariables(Emitter.NearCameraFadeRadius:NiagaraFloat=&FNTimeOfDayScalar(Night Scale=120, Morning Scale=130, Day Scale=150, Evening Scale=130)) {isSetParameters=true}]
  emitterUpdateScript [
   &EmitterState()
   &SpawnBurst_Instantaneous(Spawn Count=6, Use Spawn Probability=false)
  ]
  particleSpawnScript [
   /SetVariables(Particles.VelCurveRand:NiagaraFloat=&FloatFromCurve(
     FloatCurve=DataInterface() {propertyValues {Curve {keys [
         {interpMode=RCIM_Cubic, leaveTangent=0.1132, tangentMode=RCTM_Break, time=0, value=0}
         {arriveTangent=1.9858, interpMode=RCIM_Cubic, leaveTangent=1.7203, tangentMode=RCTM_Break, time=0.3, value=0.4}
         {interpMode=RCIM_Cubic, tangentMode=RCTM_Break, time=1, value=1}
        ]}}}
     CurveIndex=&RandomRangeFloat(Minimum=0.9, Maximum=1)
     Scale Curve=5
    )) {isSetParameters=true, module=SetVariables_#1}
   /SetVariables(Particles.InitSpriteRotRate=&RandomRangeFloat(Minimum=-1, Maximum=1)) {isSetParameters=true, module=SetVariables_#2}
   &InitializeParticle(
    Color=&SelectLinearColorFromArray(
     Array Sampling Mode=::NewEnumerator0
     Direct Array Index=?
     Color Selection Array=DataInterface() {propertyValues {ColorData [
        <rgba 0.0789 0.3385 0.2424 1>
        <rgba 0.0775 0.2188 0.0979 1>
        <rgba 0.0775 0.2188 0.0979 1>
        <rgba 0.0775 0.2188 0.0979 1>
        <rgba 0.1453 0.3021 0.0795 1>
        <rgba 0.099 0.2865 0.0462 1>
       ]}}
     Random Seed=0
    )
    Lifetime Max=1
    Lifetime Min=0.6
    Lifetime Mode=::NewEnumerator1
    Position Offset Coordinate Space=::Local
    Sprite Rotation Angle=&RandomRangeFloat(Minimum=0, Maximum=360)
    Sprite Rotation Mode=::NewEnumerator0
    Sprite Size Mode=::NewEnumerator1
    Uniform Sprite Size Max=100
    Uniform Sprite Size Min=80
    UsePositionOffset=true
    Write Material Random=true
   )
   /SetVariables(Particles.SpriteSize=&Multiply_Vector2D_ByFloat(Float=1.65, Vector2D=$Particles.SpriteSize)) {isSetParameters=true, module=SetVariables_#3}
   &CameraOffset(Camera Offset Amount=&UniformRangedFloat(Minimum=25, Maximum=50))
   &ShapeLocation(
    Handle Radius=0.1
    Large Radius=10
    MaskSpawnGroup=true
    Offset=<0 0 33.3964>
    Offset Mode=::NewEnumerator0
    Override Seed=false
    Rotation Coordinate Space=::World
    Shape Primitive=::NewEnumerator3
    UseOwnerScale=true
   )
   &AddVelocityFromPoint(
    Module.Use Falloff Distance=false
    Origin Offset=<0 0 -40.5198>
    Velocity Strength=&Lerp_Float(A=933.1072, B=898.2299, Alpha=$Particles.VelCurveRand)
   )
  ]
  particleUpdateScript [
   &ParticleState()
   &Drag(Drag=&RandomRangeFloat(Minimum=15, Maximum=5.4166), Use Linear Drag=true, Use Rotational Drag=false)
   &SpriteRotationRate(Rotation Rate=&Multiply_Float(
     A=$Particles.InitSpriteRotRate
     B=&FloatFromCurve(
      FloatCurve=DataInterface() {propertyValues {Curve {keys [
          {arriveTangent=-4.4624, interpMode=RCIM_Cubic, leaveTangent=-4.4624, tangentMode=RCTM_User, time=0, value=1}
          {arriveTangentWeight=0.1718, interpMode=RCIM_Cubic, tangentMode=RCTM_User, time=0.5, value=0.15}
         ]}}}
      Scale Curve=1000
     )
    ))
   &ScaleColor(
    Scale Alpha=&FloatFromCurve(FloatCurve=DataInterface() {propertyValues {Curve {keys [
         {arriveTangent=-1.3772, interpMode=RCIM_Cubic, leaveTangent=-1.3772, tangentMode=RCTM_User, time=0, value=0}
         {arriveTangentWeight=0.1274, interpMode=RCIM_Cubic, leaveTangentWeight=0.2616, tangentMode=RCTM_User, time=0.1541, value=1}
         {arriveTangentWeight=0.2192, interpMode=RCIM_Cubic, leaveTangentWeight=0.2077, tangentMode=RCTM_User, time=0.7488, value=1}
         {arriveTangent=0.1792, interpMode=RCIM_Cubic, leaveTangent=0.1792, tangentMode=RCTM_User, time=1.012, value=0}
        ]}}})
    Scale RGB=&MakeVectorFromLinearColorRGB(Color=&RandomRangeLinearColor(
      Minimum=&ColorFromCurve(ColorCurve=DataInterface() {propertyValues {
         AlphaCurve {keys []}
         BlueCurve {keys [<at 0.1027 3>,<at 0.1857 1 RCIM_Cubic>]}
         GreenCurve {keys [<at 0.1027 3>,<at 0.1857 1 RCIM_Cubic>]}
         RedCurve {keys [<at 0.1027 3>,<at 0.1857 1 RCIM_Cubic>]}
        }})
      Maximum=<rgba 1 1 1 1>
     ))
    ScaleA=true
    ScaleRGB=true
   )
   &ScaleSpriteSize(Scale Factor=&Vector2DFromFloat(Value=&FloatFromCurve(FloatCurve=DataInterface() {propertyValues {Curve {keys [
          {arriveTangent=6.0946, interpMode=RCIM_Cubic, leaveTangent=6.0946, tangentMode=RCTM_User, time=0, value=0.5}
          {arriveTangent=0.3473, interpMode=RCIM_Cubic, leaveTangent=0.3473, tangentMode=RCTM_User, time=0.1452, value=0.9}
          {arriveTangent=-0.003, arriveTangentWeight=0.3375, interpMode=RCIM_Cubic, leaveTangent=-0.003, tangentMode=RCTM_User, time=1.1905, value=1}
         ]}}})))
   &DynamicMaterialParameters(
    Index 0 Param 1=&Add_Float(
     A=&RandomRangeFloat(Minimum=0, Maximum=1)
     B=&FloatFromCurve(
      FloatCurve=DataInterface() {propertyValues {Curve {keys [
          {arriveTangent=2.7533, interpMode=RCIM_Cubic, leaveTangent=1.8958, tangentMode=RCTM_User, time=0, value=1}
          {arriveTangent=0.8292, interpMode=RCIM_Cubic, leaveTangent=0.8292, tangentMode=RCTM_User, time=0.321, value=1.3453}
          {arriveTangent=0.0529, interpMode=RCIM_Cubic, leaveTangent=0.2021, tangentMode=RCTM_User, time=1, value=1.5}
         ]}}}
      CurveIndex=&Add_Float(A=&Multiply_Float(A=$Particles.Initial.MaterialRandom, B=0.5), B=$Particles.NormalizedAge)
     )
    )
    Index 0 Param 2=&FloatFromCurve(FloatCurve=DataInterface() {propertyValues {Curve {keys [
         {arriveTangent=0.0453, interpMode=RCIM_Cubic, tangentMode=RCTM_Break, time=0, value=-1}
         {arriveTangent=0.0604, interpMode=RCIM_Cubic, leaveTangent=0.0604, tangentMode=RCTM_User, time=1, value=-1}
        ]}}})
    Index 0 Param 3=&FloatFromCurve(
     FloatCurve=DataInterface() {propertyValues {Curve {keys [
         {arriveTangent=0.0632, interpMode=RCIM_Cubic, leaveTangent=0.0632, tangentMode=RCTM_User, time=0, value=1}
         {
          arriveTangent=-1.6858
          arriveTangentWeight=0.0333
          interpMode=RCIM_Cubic
          leaveTangent=-1.6858
          leaveTangentWeight=0.3137
          tangentMode=RCTM_User
          time=0.05
          value=0.1
         }
         {arriveTangentWeight=0.1027, interpMode=RCIM_Cubic, tangentMode=RCTM_Break, time=0.2, value=0}
        ]}}}
     Scale Curve=&RandomRangeFloat(Minimum=0.25, Maximum=0.5)
    )
    Index 0 Param 4=&FloatFromCurve(FloatCurve=DataInterface() {propertyValues {Curve {keys [<at 0.5 1>,<at 1 0>]}}})
    Index 1 Param 1=&FloatFromCurve(
     FloatCurve=DataInterface() {propertyValues {Curve {keys [
         {interpMode=RCIM_Cubic, tangentMode=RCTM_Break, time=0, value=1}
         {arriveTangentWeight=0.0833, interpMode=RCIM_Cubic, tangentMode=RCTM_Break, time=0.25, value=0.3}
        ]}}}
     Scale Curve=&RandomRangeFloat(Minimum=1.5, Maximum=3)
    )
    Index 1 Param 2=&FloatFromCurve(
     FloatCurve=DataInterface() {propertyValues {Curve {keys [
         {interpMode=RCIM_Cubic, leaveTangent=-2.5489, tangentMode=RCTM_Break, time=0, value=1}
         {interpMode=RCIM_Cubic, leaveTangent=0.9066, tangentMode=RCTM_Break, time=1, value=0}
        ]}}}
     Scale Curve=&RandomRangeFloat(Minimum=0.01, Maximum=0.1)
    )
    Index 1 Param 3=&FloatFromCurve(FloatCurve=DataInterface() {propertyValues {Curve {keys [
         {interpMode=RCIM_Cubic, leaveTangent=6.3917, tangentMode=RCTM_Break, time=0, value=0.2}
         {arriveTangent=1.7549, interpMode=RCIM_Cubic, leaveTangent=10.6767, tangentMode=RCTM_User, time=0.25, value=1}
        ]}}})
    Index 1 Param 4=&FloatFromCurve(
     FloatCurve=DataInterface() {propertyValues {Curve {keys [{arriveTangent=-0.136, interpMode=RCIM_Cubic, leaveTangent=-0.136, tangentMode=RCTM_User, time=0, value=1}, <at 1 0.95>]}}}
     Scale Curve=&RandomRangeFloat(Minimum=1.3, Maximum=2)
    )
    Index 2 Param 1=&FloatFromCurve(FloatCurve=DataInterface() {propertyValues {Curve {keys [
         {interpMode=RCIM_Cubic, leaveTangent=4.2485, tangentMode=RCTM_Break, time=0, value=0}
         {arriveTangent=1.0548, interpMode=RCIM_Cubic, leaveTangent=1.0548, tangentMode=RCTM_User, time=0.17, value=0.5715}
         <at 1 1>
        ]}}})
    Index 2 Param 2=&FloatFromCurve(
     FloatCurve=DataInterface() {propertyValues {Curve {keys [
         {arriveTangent=1.9056, interpMode=RCIM_Cubic, leaveTangent=1.9056, tangentMode=RCTM_User, time=0, value=0.2}
         {arriveTangent=0.0259, interpMode=RCIM_Cubic, leaveTangent=0.0259, tangentMode=RCTM_User, time=1, value=1.4}
        ]}}}
     Scale Curve=&RandomRangeFloat(Minimum=0.1, Maximum=0.2)
    )
    Index 2 Param 3=&FloatFromCurve(
     FloatCurve=DataInterface() {propertyValues {Curve {keys [
         {interpMode=RCIM_Cubic, leaveTangent=0.0393, tangentMode=RCTM_Break, time=0, value=1}
         {interpMode=RCIM_Cubic, leaveTangent=11.1233, tangentMode=RCTM_Break, time=1, value=1.5}
        ]}}}
     Scale Curve=0.7
    )
    Index 2 Param 4=&FloatFromCurve(
     FloatCurve=DataInterface() {propertyValues {Curve {keys [
         {arriveTangent=-0.6867, interpMode=RCIM_Cubic, leaveTangent=-0.6867, tangentMode=RCTM_User, time=0, value=1}
         {interpMode=RCIM_Cubic, leaveTangent=-1.2956, tangentMode=RCTM_Break, time=1, value=0.75}
        ]}}}
     Scale Curve=1.5
    )
    Module.Index0Float0Write=true
    Module.Index0Float1Write=true
    Module.Index0Float2Write=true
    Module.Index0Float3Write=true
    Module.Index1Float0Write=true
    Module.Index1Float1Write=true
    Module.Index1Float2Write=true
    Module.Index1Float3Write=true
    Module.Index2Float0Write=true
    Module.Index2Float1Write=true
    Module.Index2Float2Write=true
    Module.Index2Float3Write=true
    Write Parameter Index 1=::NewEnumerator1
    Write Parameter Index 2=::NewEnumerator1
   )
   &SolveForcesAndVelocity(Clamp Velocity=false, Limit Acceleration=false)
   /SetVariables(Particles.Position=$Particles.Position) {isSetParameters=true}
   &ScaleColor(
    Scale RGB=&MakeVectorFromLinearColorRGB(Color=&MakeCustomLinearColorFromBool(
      Bool=$User.bIsFrontEndPreview
      True Color=&RandomRangeLinearColor(Minimum=<rgba 0.4 0.55 1 1>, Maximum=<rgba 0.08 0.11 0.2 1>)
      False Color=<rgba 1 1 1 1>
     ))
    ScaleA=true
    ScaleRGB=true
   ) {module=ScaleColor001}
   &BlendSpace("Position Blend (World <-> Local)":NiagaraFloat=&MakeCustomFloatFromBool(
     Bool=&SetBoolByFloatComparison(A=&VectorLength(Vector=$Engine.Owner.Velocity), B=800)
     True Float=1
     False Float=0
     Translation Contribution:NiagaraFloat=1
     Rotation Contribution:NiagaraFloat=&MakeCustomFloatFromBool(
      Bool=&SetBoolByFloatComparison(A=&VectorLength(Vector=$Engine.Owner.Velocity), B=800)
      True Float=1
      False Float=0
     )
    ))
  ]
  rendererClasses [&NiagaraSpriteRendererProperties]
  renderers [{rendererClass=&NiagaraSpriteRendererProperties}]
 }
 Smoke_Trail {
  emitterUpdateScript [&EmitterState()]
  particleSpawnScript [&InitializeParticle(
    Color=&FNTimeOfDayColor(
     Night Color=<rgba 0.0228 0.1562 0.1478 0.5826>
     Morning Color=<rgba 0.1079 0.3698 0.1457 0.3304>
     Day Color=<rgba 0.0719 0.2656 0.0999 0.5>
     Evening Color=<rgba 0.0615 0.3281 0.1 0.5>
    )
    Lifetime=3
    Ribbon Width=15
    Ribbon Width Mode=::NewEnumerator1
    UsePositionOffset=false
    Write Material Random=true
   )]
  particleUpdateScript [
   &ParticleState()
   &ScaleColor(
    Scale Alpha=&FloatFromCurve(FloatCurve=DataInterface() {propertyValues {Curve {keys [
         <at 0 1 RCIM_Cubic>
         {arriveTangent=-0.9204, interpMode=RCIM_Cubic, leaveTangent=-0.9204, tangentMode=RCTM_User, time=0.9651, value=0.75}
         {arriveTangent=-10.6106, interpMode=RCIM_Cubic, leaveTangent=-10.6106, tangentMode=RCTM_User, time=0.9862, value=0}
        ]}}})
    Scale RGB=&MakeVectorFromLinearColorRGB(Color=&ColorFromCurve(ColorCurve=DataInterface() {propertyValues {
        AlphaCurve {keys []}
        BlueCurve {keys [<at 0 1.15 RCIM_Cubic>]}
        GreenCurve {keys [<at 0 1.15 RCIM_Cubic>]}
        RedCurve {keys [<at 0 1.15 RCIM_Cubic>]}
       }}))
    ScaleA=true
    ScaleRGB=true
   )
   &ScaleColor(
    Scale Alpha=&FloatFromCurve(
     FloatCurve=DataInterface() {propertyValues {Curve {keys [
         <at 0.0852 1 RCIM_Cubic>
         {arriveTangent=-2.7188, interpMode=RCIM_Cubic, leaveTangent=-2.7188, tangentMode=RCTM_User, time=0.1782, value=0.5}
         {arriveTangent=0.0737, interpMode=RCIM_Cubic, leaveTangent=0.0737, tangentMode=RCTM_User, time=0.4334, value=0}
        ]}}}
     CurveIndex=$System.NormalizedLoopAge
     Scale Curve=&RandomRangeFloat(Minimum=1, Maximum=1)
    )
    ScaleA=true
    ScaleRGB=true
   ) {module=ScaleColor001}
   &BlendSpace("Position Blend (World <-> Local)":NiagaraFloat=&MakeCustomFloatFromBool(
     Bool=&SetBoolByFloatComparison(A=&VectorLength(Vector=$Engine.Owner.Velocity), B=800)
     True Float=1
     False Float=0
     Translation Contribution:NiagaraFloat=1
     Rotation Contribution:NiagaraFloat=&MakeCustomFloatFromBool(
      Bool=&SetBoolByFloatComparison(A=&VectorLength(Vector=$Engine.Owner.Velocity), B=800)
      True Float=1
      False Float=0
     )
    ))
   &ScaleRibbonWidth(Ribbon Width Scale=&FloatFromCurve(FloatCurve=DataInterface() {propertyValues {Curve {keys [
         {arriveTangent=48.6974, interpMode=RCIM_Cubic, leaveTangent=48.6974, tangentMode=RCTM_User, time=0, value=0.5}
         {arriveTangent=30.6462, interpMode=RCIM_Cubic, leaveTangent=30.6462, tangentMode=RCTM_User, time=0.0162, value=1}
        ]}}}))
  ]
  rendererClasses [&NiagaraRibbonRendererProperties]
  renderers [{rendererClass=&NiagaraRibbonRendererProperties}]
 }
 Streaks {
  emitterSpawnScript [/SetVariables(Emitter.NearCameraFade_Distance=190) {isSetParameters=true}]
  emitterUpdateScript [
   &EmitterState()
   &SpawnBurst_Instantaneous(Spawn Count=7, Use Spawn Probability=false, UseLoopCountLimit=false)
   &SpawnBurst_Instantaneous(Spawn Count=7, Spawn Time=0.05, Use Spawn Probability=false, UseLoopCountLimit=false) {module=SpawnBurst_Instantaneous002}
  ]
  particleSpawnScript [
   &InitializeParticle(
    Color=&SelectLinearColorFromArray(
     Array Sampling Mode=::NewEnumerator0
     Direct Array Index=?
     Color Selection Array=DataInterface() {propertyValues {ColorData [<rgba 0.0634 0.2135 0.0851 1>, <rgba 0.0557 0.1875 0.0747 1>, <rgba 0.0472 0.1927 0.1329 1>, <rgba 0.0768 0.2031 0.0451 1>]}}
     Random Seed=0
    )
    Lifetime Max=0.87
    Lifetime Min=0.6
    Lifetime Mode=::NewEnumerator1
    Sprite Size=&Multiply_Vector2D_ByFloat(Float=1, Vector2D=&UniformRangedVector2D(Minimum=<80 160>, Maximum=<120 110>))
    Sprite Size Mode=::NewEnumerator4
    UsePositionOffset=false
    Write Material Random=true
   )
   /Niagara/Modules/Spawn/Location/SphereLocation(
    MaskSpawnGroup=false
    Module.Override Randomness=false
    Offset=<0 0 -10>
    Sphere Radius=&Multiply_Float(A=10, B=1)
    Surface Only=true
   ) {enabled=false}
   &ShapeLocation(MaskSpawnGroup=false, Offset=<0 0 14>, Offset Mode=::NewEnumerator0, Override Seed=false, Sphere Radius=20, UseOwnerScale=true)
   &AddVelocityFromPoint(
    Module.Use Falloff Distance=false
    Origin Offset=&Multiply_VectorByFloat(Vector=<0 0 -12.2326>, Float=1)
    Velocity Strength=&Multiply_Float(A=&UniformRangedFloat(Minimum=300, Maximum=600), B=1)
   )
   &SubUVAnimation(End Frame=3, SubUV Animation Mode:ENiagara_SubUVLookupMode=::NewEnumerator2)
   &CameraOffset()
  ]
  particleUpdateScript [
   &ParticleState()
   &Drag(
    Drag=&FloatFromCurve(FloatCurve=DataInterface() {propertyValues {Curve {keys [<at 0.2 1>,<at 0.3 0>]}}}, Scale Curve=15)
    Use Linear Drag=true
    Use Rotational Drag=false
   )
   &GravityForce(Gravity=&VectorFromCurve(VectorCurve=DataInterface() {propertyValues {
       XCurve {keys [<at 0 0>,<at 1 0>]}
       YCurve {keys [<at 0 0>,<at 1 0>]}
       ZCurve {keys [
         {interpMode=RCIM_Cubic, tangentMode=RCTM_User, time=0, value=0}
         {arriveTangent=-100.6106, interpMode=RCIM_Cubic, leaveTangent=-100.6106, tangentMode=RCTM_User, time=0.498, value=-15}
        ]}
      }}))
   &SolveForcesAndVelocity(Clamp Velocity=false, Limit Acceleration=false)
   &ScaleSpriteSize(Scale Factor=&Vector2DFromCurve(
     Vector2Curve=DataInterface() {propertyValues {
       XCurve {keys [
         {arriveTangent=0.4813, interpMode=RCIM_Cubic, leaveTangent=0.4813, tangentMode=RCTM_User, time=0, value=0.625}
         {arriveTangent=1.25, interpMode=RCIM_Cubic, leaveTangent=1.25, time=0.2884, value=0.95}
         {arriveTangent=1.186, interpMode=RCIM_Cubic, leaveTangent=1.186, tangentMode=RCTM_User, time=0.9, value=1.75}
        ]}
       YCurve {keys [
         {interpMode=RCIM_Cubic, leaveTangentWeight=0.3337, tangentMode=RCTM_User, time=0, value=0}
         {arriveTangent=1.3858, interpMode=RCIM_Cubic, leaveTangent=1.3858, tangentMode=RCTM_User, time=0.2, value=1.4}
         {arriveTangent=0.0287, interpMode=RCIM_Cubic, leaveTangent=0.0287, tangentMode=RCTM_User, time=1, value=1.575}
        ]}
      }}
     ScaleCurve=&Multiply_Vector2D_ByFloat(Float=&RandomRangeFloat(Minimum=0.65, Maximum=0.75), Vector2D=<0.8 1.5>)
    ))
   &ScaleColor(
    Scale Mode=::NewEnumerator1
    Scale RGBA=&MakeVector4_FromLinearColor(LinearColor=&FNTimeOfDayColor(
      Night Color=&ColorFromCurve(ColorCurve=DataInterface() {propertyValues {
         AlphaCurve {keys [
           {arriveTangent=-1.087, interpMode=RCIM_Cubic, leaveTangent=-1.0164, tangentMode=RCTM_Break, time=0, value=0}
           {arriveTangent=-0.3577, interpMode=RCIM_Cubic, leaveTangent=-0.3577, tangentMode=RCTM_User, time=0.1047, value=3}
           <at 0.7 1.6>
           <at 1 0>
          ]}
         BlueCurve {keys [<at 0.0108 3>]}
         GreenCurve {keys [<at 0.0108 3>]}
         RedCurve {keys [<at 0.0108 3>]}
        }})
      Morning Color=&ColorFromCurve(ColorCurve=DataInterface() {propertyValues {
         AlphaCurve {keys [
           {arriveTangent=-1.087, interpMode=RCIM_Cubic, leaveTangent=-1.0164, tangentMode=RCTM_Break, time=0, value=0}
           {arriveTangent=-0.3577, interpMode=RCIM_Cubic, leaveTangent=-0.3577, tangentMode=RCTM_User, time=0.1047, value=8}
           <at 0.7 1>
           <at 1 0>
          ]}
         BlueCurve {keys [<at 0.0108 1>]}
         GreenCurve {keys [<at 0.0108 1>]}
         RedCurve {keys [<at 0.0108 1>]}
        }})
      Day Color=&ColorFromCurve(ColorCurve=DataInterface() {propertyValues {
         AlphaCurve {keys [
           {arriveTangent=-1.087, interpMode=RCIM_Cubic, leaveTangent=-1.0164, tangentMode=RCTM_Break, time=0, value=0}
           {arriveTangent=-0.3577, interpMode=RCIM_Cubic, leaveTangent=-0.3577, tangentMode=RCTM_User, time=0.1047, value=3}
           <at 0.7 1.3>
           <at 1 0>
          ]}
         BlueCurve {keys [<at 0.0108 2>]}
         GreenCurve {keys [<at 0.0108 2>]}
         RedCurve {keys [<at 0.0108 2>]}
        }})
      Evening Color=&ColorFromCurve(ColorCurve=DataInterface() {propertyValues {
         AlphaCurve {keys [
           {arriveTangent=-1.087, interpMode=RCIM_Cubic, leaveTangent=-1.0164, tangentMode=RCTM_Break, time=0, value=0}
           {arriveTangent=-0.3577, interpMode=RCIM_Cubic, leaveTangent=-0.3577, tangentMode=RCTM_User, time=0.1047, value=8}
           <at 0.7 1>
           <at 1 0>
          ]}
         BlueCurve {keys [<at 0.0108 1>]}
         GreenCurve {keys [<at 0.0108 1>]}
         RedCurve {keys [<at 0.0108 1>]}
        }})
     ))
    ScaleRGBA=true
   )
   &ScaleColor(
    Scale RGB=&MakeVectorFromLinearColorRGB(Color=&FNTimeOfDayColor(
      Night Color=<rgba 0.2266 0.3587 0.7552 1>
      Morning Color=<rgba 1 1 1 1>
      Day Color=&MakeCustomLinearColorFromBool(Bool=$User.bIsFrontEndPreview, True Color=<rgba 0.5 0.625 1 1>, False Color=<rgba 1 1 1 1>)
      Evening Color=<rgba 1 1 1 1>
     ))
    ScaleA=true
    ScaleRGB=true
   ) {module=ScaleColor002}
   &DynamicMaterialParameters(
    Index 0 Param 1=&Multiply_Float(A=10, B=1)
    Index 0 Param 4=0.45
    Index 1 Param 1=&FloatFromCurve(
     FloatCurve=DataInterface() {propertyValues {Curve {keys [
         {interpMode=RCIM_Cubic, tangentMode=RCTM_User, time=0, value=0}
         {arriveTangent=0.6295, interpMode=RCIM_Cubic, leaveTangent=0.6295, tangentMode=RCTM_User, time=1, value=1}
        ]}}}
     Scale Curve=0.2
    )
    Index 1 Param 2=&FloatFromCurve(
     FloatCurve=DataInterface() {propertyValues {Curve {keys [
         {arriveTangent=-100.8588, time=0, value=1}
         {arriveTangent=-26.7969, time=0.2007, value=0.2}
         {arriveTangent=-2.6967, leaveTangent=-2.6967, time=0.3, value=0.1}
        ]}}}
     CurveIndex=$Emitter.Age
     Scale Curve=&MaskFloatBySpawnGroup(Masked Float=0, Passthrough Float=5, Spawn Group Mask=1)
    )
    Index 1 Param 3=&FloatFromCurve(
     FloatCurve=DataInterface() {propertyValues {Curve {keys [{interpMode=RCIM_Cubic, tangentMode=RCTM_User, time=0.2, value=0.5}, {interpMode=RCIM_Cubic, tangentMode=RCTM_User, time=1, value=1}]}}}
     Scale Curve=2
    )
    Index 1 Param 4=&FloatFromCurve(FloatCurve=DataInterface() {propertyValues {Curve {keys [<at 0 0.3>,<at 1 0>]}}}, Scale Curve=-0.3)
    Module.Index0Float0Write=true
    Module.Index0Float1Write=true
    Module.Index0Float2Write=true
    Module.Index0Float3Write=true
    Module.Index1Float0Write=true
    Module.Index1Float1Write=false
    Module.Index1Float2Write=true
    Module.Index1Float3Write=true
    Write Parameter Index 1=::NewEnumerator1
   )
   &ScaleColor(
    Scale Alpha=0.7
    Scale RGB=&Multiply_VectorByFloat(
     Vector=<1 1 1>
     Float=&FN_PickFloatDependOnPlatform(TargetPlatform=DataInterface() {propertyValues {Platforms {qualityLevelMask=-1}}}, TargetPlatformActive=1, TargetPlatformNoActive=0.4)
    )
    ScaleA=true
    ScaleRGB=true
   ) {module=ScaleColor001}
   /SetVariables(Particles.Position=$Particles.Position) {isSetParameters=true}
   &BlendSpace("Position Blend (World <-> Local)":NiagaraFloat=&MakeCustomFloatFromBool(
     Bool=&SetBoolByFloatComparison(A=&VectorLength(Vector=$Engine.Owner.Velocity), B=800)
     True Float=1
     False Float=0
     Translation Contribution:NiagaraFloat=1
     Rotation Contribution:NiagaraFloat=&MakeCustomFloatFromBool(
      Bool=&SetBoolByFloatComparison(A=&VectorLength(Vector=$Engine.Owner.Velocity), B=800)
      True Float=1
      False Float=0
     )
    ))
  ]
  rendererClasses [&NiagaraSpriteRendererProperties]
  renderers [{rendererClass=&NiagaraSpriteRendererProperties}]
 }
 Trail_Head {
  emitterUpdateScript [
   &EmitterState()
   &SpawnBurst_Instantaneous(Spawn Count=6, Use Spawn Probability=false, UseLoopCountLimit=false)
   &SpawnRate(SpawnRate=10, Use Spawn Probability=false) {enabled=false}
  ]
  particleSpawnScript [
   &InitializeParticle(
    Lifetime Max=1.2
    Lifetime Min=0.9
    Lifetime Mode=::NewEnumerator1
    Position Offset=$System.NS Position
    Sprite Size Mode=::NewEnumerator1
    Uniform Sprite Size Max=0.5
    Uniform Sprite Size Min=0.3
    UsePositionOffset=true
    Write Material Random=true
   )
   &ShapeLocation(
    MaskSpawnGroup=false
    Offset=<0 0 52.132>
    Offset Mode=::NewEnumerator0
    Override Seed=false
    Rotation Coordinate Space=::World
    Sphere Radius=20
    UseOwnerScale=true
   ) {module=ShapeLocation001}
   &AddVelocity(
    Module.Use Falloff Distance=false
    Origin Offset=<0 0 35.4314>
    Override Seed=false
    Velocity Mode=::NewEnumerator1
    Velocity Speed=&RandomRangeFloat(Minimum=400, Maximum=748.2114)
   )
   &AddVelocity(Velocity=&RandomRangeVector(Minimum=<0 -100 200>, Maximum=<0 -100 500>)) {module=AddVelocity001}
  ]
  particleUpdateScript [
   &ParticleState()
   &SpriteRotationRate(Rotation Rate=&Multiply_Float(A=&RandomRangeFloat(Minimum=-1, Maximum=1), B=800))
   &Drag(
    Drag=&FloatFromCurve(
     FloatCurve=DataInterface() {propertyValues {Curve {keys [<at 0.0024 0 RCIM_Cubic>,<at 0.3975 1 RCIM_Cubic>]}}}
     Scale Curve=&RandomRangeFloat(Minimum=3, Maximum=5)
    )
    Use Linear Drag=true
    Use Rotational Drag=false
   )
   &GravityForce(Gravity=&RandomRangeVector(Minimum=<0 0 -300>, Maximum=<0 0 -600>))
   &SolveForcesAndVelocity(Clamp Velocity=false, Limit Acceleration=false)
   &ScaleSpriteSizeBySpeed(Max Scale Factor=<0.5 8>, Min Scale Factor=<0.4 5>)
   &ScaleColor(
    Scale Alpha=&FloatFromCurve(FloatCurve=DataInterface() {propertyValues {Curve {keys [<at 0 1>,<at 0.6 0.6>,<at 1 0>]}}})
    Scale RGB=&MakeVectorFromLinearColorRGB(Color=&ColorFromCurve(ColorCurve=DataInterface() {propertyValues {
        AlphaCurve {keys []}
        BlueCurve {keys [<at 0.0734 0.1512 RCIM_Cubic>,<at 0.1716 0.1409>,<at 0.4113 0.1512 RCIM_Cubic>]}
        GreenCurve {keys [<at 0.0734 0.3281 RCIM_Cubic>,<at 0.1716 0.3281>,<at 0.4113 0.3281 RCIM_Cubic>]}
        RedCurve {keys [<at 0.0734 0.1213 RCIM_Cubic>,<at 0.1716 0.1094>,<at 0.4113 0.1213 RCIM_Cubic>]}
       }}))
    ScaleA=true
    ScaleRGB=true
   )
   &GenerateLocationEvent(Use Event Probability=false, UseEventDelay=false)
   &BlendSpace("Position Blend (World <-> Local)":NiagaraFloat=&MakeCustomFloatFromBool(
     Bool=&SetBoolByFloatComparison(A=&VectorLength(Vector=$Engine.Owner.Velocity), B=800)
     True Float=1
     False Float=0
     Translation Contribution:NiagaraFloat=1
     Rotation Contribution:NiagaraFloat=&MakeCustomFloatFromBool(
      Bool=&SetBoolByFloatComparison(A=&VectorLength(Vector=$Engine.Owner.Velocity), B=800)
      True Float=1
      False Float=0
     )
    ))
  ]
  rendererClasses [&NiagaraSpriteRendererProperties]
  renderers [{rendererClass=&NiagaraSpriteRendererProperties}]
 }
}
summary {
 systemName=NS_Emote_CanineCronutMix_PowderExplosion
 emitters [
  {emitterName=Streaks, rendererClasses [&NiagaraSpriteRendererProperties]}
  {emitterName=Smoke_Poof_Out, rendererClasses [&NiagaraSpriteRendererProperties]}
  {emitterName=Smoke_Poof_Up, rendererClasses [&NiagaraSpriteRendererProperties]}
  {emitterName=Smoke_Trail, rendererClasses [&NiagaraRibbonRendererProperties]}
  {emitterName=Trail_Head, rendererClasses [&NiagaraSpriteRendererProperties]}
  {
   emitterName=Flare_Dusty_Decal_ReVamped
   rendererClasses [&NiagaraSpriteRendererProperties, &NiagaraDecalRendererProperties]
  }
  {emitterName=FallingSmokeDust, rendererClasses []}
 ]
 userVariables [
  User.bIsFrontEnd {type=NiagaraBool}
  User.bIsFrontEndPreview {type=NiagaraBool}
  User.IsInWater {type=NiagaraBool}
  User.Trail {type=NiagaraFloat, defaultValue=0, description=""}
  User.Trail2 {type=NiagaraFloat, defaultValue=0, description=""}
  User.VariantSwitch {type=NiagaraFloat, defaultValue=0, description=""}
 ]
}
